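# Material function for the VelvetDeskFiber character body that adds a staged, animated heat and fire look to base color, metallic, roughness and emissive.
# CurrentStage and a Blueprint-supplied start time drive per-stage lerps, some shaped by curve atlas rows.
# Panning turbulence projected on pre-skinned position within a sphere mask, a heat-gradient lookup, hot spots and fake anodized reflections build the glow.
# TechArt and texture masks supply masking, and EnableFireVisual statically toggles the effect.
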MaterialFunction /BRCosmetics/Characters/Player/Male/Medium/Bodies/M_MED_VelvetDeskFiber/Materials/MF_VelvetDeskFiber
graph {nodes [
  n1 = GetMaterialAttributes(None=FunctionInput() {inputType=FunctionInput_MaterialAttributes}) {attributeGetTypes [
    #1
    #2
    #3
    #4
    #5
    #6
   ]}
  n2 = SetMaterialAttributes!(
   MaterialAttributes=@n1.MaterialAttributes
   BaseColor=NamedRerouteUsage()
   Metallic=NamedRerouteUsage()
   Specular=@n1.MaterialAttributes
   Roughness=NamedRerouteUsage()
   EmissiveColor=NamedRerouteUsage()
   Normal=@n1.MaterialAttributes
  ) {attributeSetTypes [
    #1
    #2
    #3
    #4
    #6
    #5
   ]}
  FunctionOutput(None=@n2) {outputName=OUT}
  NamedRerouteDeclaration(None=@n1.Normal) {name=IN_Normal, nodeColor=<rgba 0 1 0.6719 1>}
  n3 = NamedRerouteUsage()
  n4 = NamedRerouteUsage()
  n5 = LinearInterpolate(A=@n3, B=LinearInterpolate(A=@n3, B=Add(A=@n3, B=@n4), Alpha=NamedRerouteUsage()), Alpha=NamedRerouteUsage())
  n6 = Multiply(A=ScalarParameter() {defaultValue=0.125, group=VelvetDeskFiber, parameterName=HeatIntensity, sliderMax=1}, B=NamedRerouteUsage())
  NamedRerouteDeclaration(None=StaticSwitchParameter(True=LinearInterpolate(A=@n3, B=@n5, Alpha=@n6), False=@n3) {group=VelvetDeskFiber, parameterName=EnableFireVisual}) {name=OUT_BC, nodeColor=<rgba 1 0.3516 0 1>}
  NamedRerouteDeclaration(None=VertexInterpolator(VS=PreSkinnedNormal()).PS) {name=PreSkinnedNormal, nodeColor=<rgba 0 1 0.6719 1>}
  n7 = Subtract(A=Time(), B=ScalarParameter() {defaultValue=-99999, group=VelvetDeskFiber, parameterName=GameTimeFromBP, sortPriority=100})
  n8 = Add(A=ScalarParameter() {defaultValue=1, group=VelvetDeskFiber, parameterName=Animation_Duration, sortPriority=101}) {constB=0}
  NamedRerouteDeclaration(None=Saturate(None=Divide(A=@n7, B=@n8))) {name=TimeNative, nodeColor=<rgba 0.2891 1 0 1>}
  n9 = FunctionInput() {inputName=TechArtMask, sortPriority=10}
  n10 = DotProduct(A=@n9, B=@n9)
  n11 = StaticComponentMaskParameter(None=AppendVector(A=@n9, B=OneMinus(None=@n10)) {desc=RestArea to A Channel}) {defaultB=true, defaultG=true, defaultR=true, group=VelvetDeskFiber, parameterName=SelectedTAMaskForApplyFX}
  NamedRerouteDeclaration(None=DotProduct(A=@n11, B=@n11)) {name=ArmorMask, nodeColor=<rgba 0.0547 1 0 1>}
  NamedRerouteDeclaration(None=FunctionInput() {inputName=Frenel, sortPriority=10}) {name=Fresnel_IN, nodeColor=<rgba 1 0 0.1172 1>}
  n12 = Subtract(A=NamedRerouteUsage(), B=VectorParameter() {group=VelvetDeskFiber, parameterName=PossitionOffset, defaultValue=<rgba -50 8 100 1>}.RGB)
  NamedRerouteDeclaration(None=@n12) {name=OffsetPreskinnedPos, nodeColor=<rgba 1 0 0.9844 1>}
  n13 = ScalarParameter() {group=VelvetDeskFiber, parameterName=CurrentStage, sortPriority=3}
  n14 = LinearInterpolate(A=Subtract(A=@n13), B=@n13, Alpha=NamedRerouteUsage()) {desc=Native lerp toward}
  n15 = Constant4Vector() {desc=Stage 1 to 2, constant=<rgba 1 2 0 1>}
  NamedRerouteDeclaration(None=Saturate(None=MaterialFunctionCall(Input=@n14, Input Low=@n15.R, Input High=@n15.R, Target Low=@n15.R, Target High=@n15.R) {materialFunction=&RemapValueRange}.Result)) {name=Stage2_Lerp, nodeColor=<rgba 0.1893 1 0 1>}
  n16 = Multiply(A=Time(), B=ScalarParameter() {defaultValue=8, group=VelvetDeskFiber, parameterName=HeatNoiseSpeed}) {constB=2}
  n17 = VectorParameter() {group=VelvetDeskFiber, parameterName=HeatNoisePanningDirection, defaultValue=<rgba 0 -0.1 -1 1>}
  n18 = MaterialFunctionCall(
   TextureObject=TextureObjectParameter() {group=VelvetDeskFiber, parameterName=HeatNoise, texture=/Game/Global/Textures/Default/Caustics/T_Turbulence_Seamless}
   TextureSize=VectorParameter() {group=VelvetDeskFiber, parameterName=HeatNoiseScale, defaultValue=<rgba 20 20 20 1>}.RGB
   WorldPosition=Add(A=@n12, B=Multiply(A=@n16, B=@n17.RGB))
   World Space Normal=NamedRerouteUsage()
   ProjectionTransitionContrast=Constant() {r=1}
  ) {materialFunction=&WorldAlignedTexture}
  n19 = ScalarParameter() {defaultValue=1, group=VelvetDeskFiber, parameterName=HeatNoiseIntensity}
  n20 = SphereMask(A=@n12, B=Constant(), Radius=ScalarParameter() {defaultValue=40, group=VelvetDeskFiber, parameterName=SphereMaskRadius}) {attenuationRadius=40, hardnessPercent=1}
  NamedRerouteDeclaration(None=Multiply(A=Multiply(A=@n18.XYZ Texture, B=@n19), B=@n20)) {name=NoiseMask, nodeColor=<rgba 0 1 0.6953 1>}
  n21 = Constant4Vector() {desc=Stage 2 to 3, constant=<rgba 2 3 0 1>}
  n22 = CurveAtlasRowParameter(CurveTime=Saturate(None=MaterialFunctionCall(Input=@n14, Input Low=@n21.R, Input High=@n21.R, Target Low=@n21.R, Target High=@n21.R) {materialFunction=&RemapValueRange}.Result)) {
   atlas=/BRCosmetics/Characters/Player/Male/Medium/Bodies/M_MED_VelvetDeskFiber/Textures/CurveAtlas_VelvetDeskFiber
   curve=/BRCosmetics/Characters/Player/Male/Medium/Bodies/M_MED_VelvetDeskFiber/Textures/Curve_VelvetDeskFiber_Animation01
   group=VelvetDeskFiber
   parameterName=AnimationCurveAtlas
  }
  NamedRerouteDeclaration(None=@n22.G) {name=Stage3_Lerp, nodeColor=<rgba 0.1893 1 0 1>}
  n23 = Constant4Vector() {desc=Stage 3 to 4, constant=<rgba 4 5 0 1>}
  NamedRerouteDeclaration(None=Saturate(None=MaterialFunctionCall(Input=@n14, Input Low=@n23.R, Input High=@n23.R, Target Low=@n23.R, Target High=@n23.R) {materialFunction=&RemapValueRange}.Result)) {name=Stage4_Lerp, nodeColor=<rgba 0.1893 1 0 1>}
  NamedRerouteDeclaration(None=@n20) {name=SphereMask, nodeColor=<rgba 0.7109 0 1 1>}
  n24 = NamedRerouteUsage()
  n25 = LinearInterpolate(A=@n24, Alpha=Multiply(A=NamedRerouteUsage(), B=NamedRerouteUsage())) {constAlpha=1, desc=rougness}
  n26 = NamedRerouteUsage()
  NamedRerouteDeclaration(None=StaticSwitchParameter(True=LinearInterpolate(A=@n24, B=@n25, Alpha=@n26), False=@n24) {group=VelvetDeskFiber, parameterName=EnableFireVisual}) {name=OUT_Roughness, nodeColor=<rgba 0.0312 0 1 1>}
  n27 = Multiply(A=NamedRerouteUsage(), B=NamedRerouteUsage())
  n28 = ScalarParameter() {defaultValue=1, group=VelvetDeskFiber, parameterName=EmissiveMapIntesity}
  n29 = Add(A=Multiply(A=Multiply(A=@n27, B=@n28), B=NamedRerouteUsage()), B=Multiply(A=NamedRerouteUsage(), B=NamedRerouteUsage()))
  n30 = Multiply(A=NamedRerouteUsage(), B=OneMinus(None=NamedRerouteUsage()))
  n31 = Transform(None=NamedRerouteUsage())
  n32 = Power(Base=DotProduct(A=CameraVectorWS(), B=Normalize(VectorInput=@n31))) {constExponent=40}
  n33 = Multiply(A=ScalarParameter() {defaultValue=1, group=VelvetDeskFiber, parameterName=HeatPointIntensity}, B=NamedRerouteUsage())
  n34 = Fresnel(
   ExponentIn=ScalarParameter() {defaultValue=1, group=VelvetDeskFiber, parameterName=Anodized_FresnelExponent}
   BaseReflectFractionIn=ScalarParameter() {defaultValue=2, group=VelvetDeskFiber, parameterName=Anodized_ReflectFraction}
   Normal=PixelNormalWS()
  ) {baseReflectFraction=0.5, exponent=3}
  n35 = ScalarParameter() {group=VelvetDeskFiber, parameterName=Anodized_ColorOffset}
  n36 = Multiply(
   A=TextureSample(UVs=Add(A=@n34, B=@n35)) {texture=/Game/Global/Textures/Default/Masks_and_Gradients/3pxRGBImage}.RGB
   B=VectorParameter() {group=VelvetDeskFiber, parameterName=AnodizationFilter, defaultValue=<rgba 1 0 0.5305 0>}.RGB
  )
  n37 = Multiply(A=@n36, B=ScalarParameter() {defaultValue=4, group=VelvetDeskFiber, parameterName=AnodizedLightVector_Strength})
  n38 = StaticComponentMaskParameter(None=AtmosphericLightVector()) {
   defaultB=true
   defaultG=true
   defaultR=true
   desc=Use RG for 2D vectors dot product. else use RGB
   group=VelvetDeskFiber
   parameterName=AnodizedLdotR_Components
  }
  n39 = StaticComponentMaskParameter(None=ReflectionVectorWS()) {
   defaultB=true
   defaultG=true
   defaultR=true
   desc=Use RG for 2D vectors dot product. else use RGB
   group=VelvetDeskFiber
   parameterName=AnodizedLdotR_Components
  }
  n40 = DotProduct(A=@n38, B=@n39)
  n41 = ScalarParameter() {defaultValue=55, group=VelvetDeskFiber, parameterName=Anodized_pow}
  n42 = Saturate(None=Multiply(A=Power(Base=@n40, Exp=@n41), B=Constant() {r=512}))
  n43 = Multiply(
   A=Desaturation(None=LinearInterpolate(A=@n36, B=@n37, Alpha=@n42), Fraction=ScalarParameter() {group=VelvetDeskFiber, parameterName=Anodized_Desat})
   B=Power(Base=NamedRerouteUsage(), Exp=ScalarParameter() {defaultValue=1, group=VelvetDeskFiber, parameterName=AnodizedEffectFresnelMask})
  )
  n44 = ScalarParameter() {defaultValue=0.2, group=VelvetDeskFiber, parameterName=AnodizedEffectIntensity}
  n45 = Multiply(
   A=Add(A=Multiply(A=@n32, B=@n33), B=Multiply(A=@n43, B=@n44))
   B=VectorParameter() {group=VelvetDeskFiber, parameterName=HeatPointColor, defaultValue=<rgba 1 0.1 0 1>}.RGB
  )
  n46 = NamedRerouteUsage()
  n47 = ScalarParameter() {defaultValue=0.125, group=VelvetDeskFiber, parameterName=HeatIntensity}
  n48 = Multiply(
   A=Multiply(A=Add(A=@n45, B=@n46), B=NamedRerouteUsage()) {constA=1, desc=emissive}
   B=LinearInterpolate(A=@n47, B=Multiply(A=@n47) {constB=2}, Alpha=NamedRerouteUsage())
  ) {constA=1, desc=emissive}
  n49 = Multiply(A=Multiply(A=Add(A=@n30, B=@n48), B=NamedRerouteUsage()), B=NamedRerouteUsage())
  NamedRerouteDeclaration(None=StaticSwitchParameter(True=Add(A=@n29, B=@n49), False=@n29) {group=VelvetDeskFiber, parameterName=EnableFireVisual}) {name=OUT_Emissive, nodeColor=<rgba 1 0 0.6094 1>}
  NamedRerouteDeclaration(None=@n22.B) {name=Stage3_Lerp_2, nodeColor=<rgba 0.1893 1 0 1>}
  n50 = TextureSampleParameter2D() {
   group=VelvetDeskFiber
   parameterName=AdittionalEmissiveMask
   samplerType=SAMPLERTYPE_Masks
   texture=&T_BlackMask
  }
  NamedRerouteDeclaration(None=Multiply(A=@n50.R, B=Constant3Vector() {constant=<rgba 1 0 0 0>})) {name=AddtionalRedEmissive, nodeColor=<rgba 1 0 0.2344 1>}
  n51 = Constant4Vector() {desc=Stage 4 to 5, constant=<rgba 4 5 0 1>}
  NamedRerouteDeclaration(None=Saturate(None=MaterialFunctionCall(Input=@n14, Input Low=@n51.R, Input High=@n51.R, Target Low=@n51.R, Target High=@n51.R) {materialFunction=&RemapValueRange}.Result)) {name=Stage5_Lerp, nodeColor=<rgba 0.1893 1 0 1>}
  NamedRerouteDeclaration(None=@n1.EmissiveColor) {name=IN_Emissive, nodeColor=<rgba 0 0.8125 1 1>}
  n52 = NamedRerouteUsage()
  n53 = DotProduct(A=Transform(None=NamedRerouteUsage()), B=Normalize(VectorInput=@n52))
  n54 = LinearInterpolate(
   A=Length(None=@n52) {is=class MaterialExpressionLength}
   B=Saturate(None=Multiply(A=@n53) {constB=-1, desc=*-1 invert Dot Product})
   Alpha=Constant() {desc=0.5, r=0.5}
  ) {desc=Distance}
  n55 = ScalarParameter() {defaultValue=0.5, group=VelvetDeskFiber, parameterName=Hot Spot Size Exponent, sortPriority=21}
  n56 = Saturate(None=Divide(B=Multiply(A=@n54, B=@n55) {desc=*FalloffDamp}))
  n57 = ScalarParameter() {defaultValue=2, group=VelvetDeskFiber, parameterName=Hot Spot Falloff Exponent, sliderMax=2, sortPriority=20}
  n58 = Add(
   A=Multiply(
    A=Power(Base=@n56, Exp=@n57)
    B=ScalarParameter() {defaultValue=0.75, group=VelvetDeskFiber, parameterName=Emissive Hotspot Multiplier, sliderMax=1, sortPriority=19}
   )
   B=ScalarParameter() {defaultValue=0.005, group=VelvetDeskFiber, parameterName=Emissive Min Intensity, sliderMax=1, sortPriority=15}
  )
  n59 = NamedRerouteUsage()
  n60 = Multiply(A=Saturate(None=Multiply(A=@n58, B=@n59)), B=NamedRerouteUsage())
  n61 = DotProduct(
   A=Transform(None=NamedRerouteUsage())
   B=MaterialFunctionCall() {materialFunction=&CameraDirectionVector}.Result
  )
  n62 = Sine(None=Multiply(A=Abs(None=@n61)) {constB=3.14})
  n63 = Power(
   Base=Saturate(None=OneMinus(None=@n62))
   Exp=ScalarParameter() {defaultValue=1, group=VelvetDeskFiber, parameterName=Emissive Body Falloff, sliderMax=1, sortPriority=17}
  )
  n64 = ScalarParameter() {defaultValue=-0.005, group=VelvetDeskFiber, parameterName=Emissive Body Mulitplier, sliderMax=1, sliderMin=-1, sortPriority=16}
  n65 = Add(A=Add(A=@n60, B=Multiply(A=@n63, B=@n64))) {constB=0.01}
  n66 = Divide(A=Constant() {r=1}, B=Ceil(None=Constant() {r=1}))
  n67 = Multiply(A=@n66) {constB=0.5}
  n68 = Multiply(A=@n66, B=Ceil(None=Constant()))
  n69 = TextureSampleParameter2D(UVs=AppendVector(A=ComponentMask(None=@n65) {r=true}, B=Add(A=@n67, B=@n68))) {
   desc=Sampler Source: Clamp
   group=VelvetDeskFiber
   parameterName=HeatGradient
   samplerSource=SSM_Clamp_WorldGroupSettings
   texture=/BRCosmetics/Accessories/FORT_Backpacks/Backpack_M_MED_VelvetDeskFiber/Textures/T_Heat_Gradient
  }
  NamedRerouteDeclaration(None=@n69.RGB) {name=HeatEmissive, nodeColor=<rgba 1 0.0536 0 1>}
  NamedRerouteDeclaration(None=@n50.G) {name=FlameMask, nodeColor=<rgba 0 0.5547 1 1>}
  NamedRerouteDeclaration(None=@n60) {name=HeatPointNoise, nodeColor=<rgba 0.8516 1 0 1>}
  n70 = Constant4Vector() {desc=Stage 0 to 1, constant=<rgba 0 1 0 1>}
  n71 = Saturate(None=MaterialFunctionCall(Input=@n14, Input Low=@n70.R, Input High=@n70.R, Target Low=@n70.R, Target High=@n70.R) {materialFunction=&RemapValueRange}.Result)
  NamedRerouteDeclaration(None=@n71) {name=Stage1_Lerp_2, nodeColor=<rgba 0.1893 1 0 1>}
  NamedRerouteDeclaration(None=@n50.B) {name=InEmisiveMapMask, nodeColor=<rgba 0.4062 0 1 1>}
  n72 = NamedRerouteUsage()
  n73 = LinearInterpolate(A=@n72, B=LinearInterpolate(A=@n72, B=NamedRerouteUsage(), Alpha=NamedRerouteUsage()) {desc=metalic}, Alpha=NamedRerouteUsage())
  n74 = NamedRerouteUsage()
  NamedRerouteDeclaration(None=StaticSwitchParameter(True=LinearInterpolate(A=@n72, B=@n73, Alpha=@n74) {desc=metalic}, False=@n72) {group=VelvetDeskFiber, parameterName=EnableFireVisual}) {name=OUT_Metalic, nodeColor=<rgba 0.2891 1 0 1>}
  NamedRerouteDeclaration(None=VertexInterpolator(VS=LocalPosition() {localOrigin=InstancePreSkinning}.XYZ).PS) {name=PreskinnedPos, nodeColor=<rgba 1 0.2109 0 1>}
  n75 = CurveAtlasRowParameter(CurveTime=@n71) {
   atlas=&CurveAtlas_VelvetDeskFiber
   curve=&Curve_VelvetDeskFiber_Animation01
   group=VelvetDeskFiber
   parameterName=AnimationCurveAtlas
  }
  NamedRerouteDeclaration(None=@n75.R) {name=Stage1_Lerp, nodeColor=<rgba 0.1893 1 0 1>}
  NamedRerouteDeclaration(None=@n1.BaseColor) {name=IN_BC, nodeColor=<rgba 0.2656 0 1 1>}
  NamedRerouteDeclaration(None=@n1.Metallic) {name=IN_Metallic, nodeColor=<rgba 0.0547 1 0 1>}
  NamedRerouteDeclaration(None=@n1.Specular) {name=IN_Specular, nodeColor=<rgba 1 0 0.375 1>}
  NamedRerouteDeclaration(None=@n1.Roughness) {name=IN_Roughness, nodeColor=<rgba 0 0.6953 1 1>}
  NamedRerouteDeclaration(None=FunctionInput() {inputName=Masks, sortPriority=10}) {name=InMasks, nodeColor=<rgba 0.5703 0 1 1>}
 ]}
