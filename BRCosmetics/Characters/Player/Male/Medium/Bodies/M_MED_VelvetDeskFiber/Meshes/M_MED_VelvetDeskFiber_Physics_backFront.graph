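# Physics asset for the back and front armor flaps of the M_MED_VelvetDeskFiber male player cosmetic body.
# Capsule bodies form a two-link front chain and three-link back chain, each hung from a small base anchor tied to spine_01.
# Tight swing and twist limits along each chain allow only slight sway, with the front swinging wider.
# Pelvis, spine and thigh capsules, joined by free-rotation constraints, provide body volumes around the flaps.
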PhysicsAsset /BRCosmetics/Characters/Player/Male/Medium/Bodies/M_MED_VelvetDeskFiber/Meshes/M_MED_VelvetDeskFiber_Physics_backFront
bodies {
 base_backArmor_01 [{shapeType=Capsule, length=1, radius=0.5, shapeName=base_backArmor_01_capsule, rotation=<rot 0 -90 0>}]
 base_frontArmor_01 [{shapeType=Capsule, length=1, radius=0.5, shapeName=base_frontArmor_01_capsule, rotation=<rot 0 -90 0>}]
 dyn_backArmor_01 [{shapeType=Capsule, length=8.6639, radius=2.476, shapeName=dyn_frontArmor_02_capsule, center=<4.2765 0 1.8008>, rotation=<rot 0 0 90>}]
 dyn_backArmor_02 [{shapeType=Capsule, length=8.478, radius=2.7761, shapeName=dyn_frontArmor_03_capsule, center=<3.9802 0 2.1367>, rotation=<rot 0 0 90>}]
 dyn_backArmor_03 [{shapeType=Capsule, length=7.8365, radius=2.9328, shapeName=dyn_frontArmor_04_capsule, center=<3.5726 0 1.6265>, rotation=<rot 0 0 90>}]
 dyn_frontArmor_01 [{shapeType=Capsule, length=16.3466, radius=2.7122, shapeName=dyn_frontArmor_01_capsule, center=<5.0563 0 0.8685>, rotation=<rot 0 0 90>}]
 dyn_frontArmor_02 [
  {shapeType=Capsule, length=15.1953, radius=2.7761, shapeName=dyn_frontArmor_02_capsule, center=<3.1832 0 1.0689>, rotation=<rot 0 0 90>}
  {shapeType=Capsule, length=12.9259, radius=3.4378, shapeName=dyn_frontArmor_02_capsuleAlt, center=<7.9388 0 0.7133>, rotation=<rot 0 0 90>}
 ]
 pelvis [{shapeType=Capsule, length=8.3135, radius=10.0147, shapeName=pelvis_capsule, center=<-5.0608 -1.9258 0>}]
 spine_01 [{shapeType=Capsule, length=4.2436, radius=8.8967, shapeName=spine_01_capsule}]
 thigh_l [{
   shapeType=Capsule
   length=19.1666
   radius=9.5695
   shapeName=thigh_l_capsule
   center=<-16.3349 -0.6343 0.035>
   rotation=<rot 0.1557 87.253 -86.7581>
  }]
 thigh_r [{
   shapeType=Capsule
   length=19.5897
   radius=9.5964
   shapeName=thigh_l_capsule
   center=<16.3349 0.6343 -0.035>
   rotation=<rot 0.1557 87.253 93.2419>
  }]
}
constraints [
 {bone1Name=dyn_frontArmor_02, bone2Name=dyn_frontArmor_01, swing1LimitDegrees=6, swing2LimitDegrees=15, twistLimitDegrees=3}
 {bone1Name=dyn_frontArmor_01, bone2Name=base_frontArmor_01, swing1LimitDegrees=3, swing2LimitDegrees=20, twistLimitDegrees=3}
 {bone1Name=dyn_backArmor_01, bone2Name=base_backArmor_01, swing1LimitDegrees=2.5, swing2LimitDegrees=5, twistLimitDegrees=1}
 {bone1Name=dyn_backArmor_02, bone2Name=dyn_backArmor_01, swing1LimitDegrees=5, swing2LimitDegrees=5, twistLimitDegrees=1}
 {bone1Name=dyn_backArmor_03, bone2Name=dyn_backArmor_02, swing1LimitDegrees=5, swing2LimitDegrees=5, twistLimitDegrees=1}
 {
  bone1Name=base_frontArmor_01
  bone2Name=spine_01
  swing1LimitDegrees=45
  swing1Motion=Free
  swing2LimitDegrees=45
  swing2Motion=Free
  twistLimitDegrees=45
  twistMotion=Free
 }
 {
  bone1Name=base_backArmor_01
  bone2Name=spine_01
  swing1LimitDegrees=45
  swing1Motion=Free
  swing2LimitDegrees=45
  swing2Motion=Free
  twistLimitDegrees=45
  twistMotion=Free
 }
 {
  bone1Name=spine_01
  bone2Name=pelvis
  swing1LimitDegrees=45
  swing1Motion=Free
  swing2LimitDegrees=45
  swing2Motion=Free
  twistLimitDegrees=45
  twistMotion=Free
 }
 {
  bone1Name=thigh_l
  bone2Name=pelvis
  swing1LimitDegrees=45
  swing1Motion=Free
  swing2LimitDegrees=45
  swing2Motion=Free
  twistLimitDegrees=45
  twistMotion=Free
 }
 {
  bone1Name=thigh_r
  bone2Name=pelvis
  swing1LimitDegrees=45
  swing1Motion=Free
  swing2LimitDegrees=45
  swing2Motion=Free
  twistLimitDegrees=45
  twistMotion=Free
 }
]
