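# Physics asset for the left and right leg armor of the M_MED_VelvetDeskFiber male body cosmetic, apparently a variant for emotes.
# Three-bone dyn_legArmor chains hang from small base bodies linked to spine_01, with wider swing limits at the root and tight limits down the chain.
# Capsules on pelvis, spine and thighs supply body volumes around the armor.
# Body joints have free rotation, except thigh_r to pelvis, which has tight limits unlike thigh_l.
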PhysicsAsset /BRCosmetics/Characters/Player/Male/Medium/Bodies/M_MED_VelvetDeskFiber/Meshes/M_MED_VelvetDeskFiber_Physics_sidesEmote
bodies {
 base_legArmor_l_01 [{shapeType=Capsule, length=0.505, radius=0.505, shapeName=base_legArmor_l_01_capsule, rotation=<rot 0 -90 0>}]
 base_legArmor_r_01 [{shapeType=Capsule, length=0.505, radius=0.505, shapeName=base_legArmor_r_01_capsule, rotation=<rot 0 -90 0>}]
 dyn_legArmor_l_01 [
  {
   shapeType=Capsule
   length=7.625
   radius=3.4786
   shapeName=dyn_legArmor_l_01_capsule
   center=<5.5023 11.4337 -4.0913>
   rotation=<rot 89.8222 -179.977 -179.9772>
  }
  {
   shapeType=Capsule
   length=7.625
   radius=3.4786
   shapeName=dyn_legArmor_l_01_capsule
   center=<4.7546 0 4.288>
   rotation=<rot 89.8222 -179.977 -179.9772>
  }
 ]
 dyn_legArmor_l_02 [
  {
   shapeType=Capsule
   length=11.8544
   radius=4.3422
   shapeName=dyn_legArmor_l_02_capsule
   center=<6.3812 11.602 -3.9251>
   rotation=<rot -83 -170 177>
  }
  {
   shapeType=Capsule
   length=8.4565
   radius=3.637
   shapeName=dyn_legArmor_l_02_capsule
   center=<4.7192 -10.6472 -2.3056>
   rotation=<rot -88.4 -9.9995 -0.0005>
  }
  {
   shapeType=Capsule
   length=11.259
   radius=3.8146
   shapeName=dyn_legArmor_l_02_capsule
   center=<6.7087 0 3.4948>
   rotation=<rot -86.7032 179.9995 -179.9994>
  }
 ]
 dyn_legArmor_l_03 [
  {
   shapeType=Capsule
   length=10.8823
   radius=4.5459
   shapeName=dyn_legArmor_l_03_capsule
   center=<9.3935 -7.9784 0.6904>
   rotation=<rot 4.1722 5.5892 -122.8525>
  }
  {shapeType=Capsule, length=9.4876, radius=4.743, shapeName=dyn_legArmor_l_03_capsule, center=<8.8048 8.173 1.0316>, rotation=<rot 0 0 120>}
 ]
 dyn_legArmor_r_01 [{
   shapeType=Capsule
   length=7.625
   radius=2.805
   shapeName=dyn_legArmor_l_01_capsule
   center=<4.7546 0 4.288>
   rotation=<rot 89.8222 -179.977 0.0228>
  }]
 dyn_legArmor_r_02 [
  {shapeType=Capsule, length=11.6, radius=4.0368, shapeName=dyn_legArmor_l_02_capsule, center=<6.7087 0 3.4948>, rotation=<rot -88.4 180 0>}
  {
   shapeType=Capsule
   length=12.0035
   radius=4.3616
   shapeName=dyn_legArmor_l_02_capsule
   center=<6.8146 -11.2521 -0.7157>
   rotation=<rot -83 -170 -17>
  }
  {
   shapeType=Capsule
   length=8.1776
   radius=3.4889
   shapeName=dyn_legArmor_l_02_capsule
   center=<5.5286 10.7899 -2.4269>
   rotation=<rot -81.6 -169.9999 -10.0001>
  }
 ]
 dyn_legArmor_r_03 [
  {
   shapeType=Capsule
   length=8.2911
   radius=3.7109
   shapeName=dyn_legArmor_l_03_capsule
   center=<9.0629 7.5856 -0.1577>
   rotation=<rot 3.5382 -5.9778 -59.9473>
  }
  {
   shapeType=Capsule
   length=10.564
   radius=3.9339
   shapeName=dyn_legArmor_l_03_capsule
   center=<8.9581 -7.8297 0.4371>
   rotation=<rot 2.4976 4.3328 60.0945>
  }
 ]
 pelvis [{shapeType=Capsule, length=8.3135, radius=10.0147, shapeName=pelvis_capsule, center=<-5.0608 -1.9258 0>}]
 spine_01 [{shapeType=Capsule, length=7.7633, radius=8.8967, shapeName=spine_01_capsule}]
 thigh_l [
  {
   shapeType=Capsule
   length=9.8359
   radius=10.0773
   shapeName=thigh_l_capsule
   center=<-10.9328 -0.8081 0.841>
   rotation=<rot 0.1556 87.2523 -85.6174>
  }
  {
   shapeType=Capsule
   length=14.7623
   radius=8.4647
   shapeName=thigh_l_capsule
   center=<-29.3213 -0.2707 1.0969>
   rotation=<rot 0.4219 90.3878 -96.0244>
  }
 ]
 thigh_r [
  {
   shapeType=Capsule
   length=9.8359
   radius=10.0773
   shapeName=thigh_l_capsule
   center=<10.9328 0.8081 -0.841>
   rotation=<rot 0.1556 87.2523 94.3826>
  }
  {
   shapeType=Capsule
   length=14.7623
   radius=8.4647
   shapeName=thigh_l_capsule
   center=<29.3213 0.2707 -1.0969>
   rotation=<rot 0.4219 90.3878 83.9756>
  }
 ]
 thigh_twist_01_l [{shapeType=Capsule, length=17.8506, radius=9.8394, shapeName=thigh_twist_01_l_capsule, rotation=<rot 90 0 0>}]
 thigh_twist_01_r [{shapeType=Capsule, length=17.7972, radius=9.2746, shapeName=thigh_twist_01_r_capsule, rotation=<rot 90 0 0>}]
}
constraints [
 {bone1Name=dyn_legArmor_l_02, bone2Name=dyn_legArmor_l_01, swing1LimitDegrees=5, swing2LimitDegrees=10, twistLimitDegrees=1}
 {bone1Name=dyn_legArmor_l_03, bone2Name=dyn_legArmor_l_02, swing1LimitDegrees=5, swing2LimitDegrees=5, twistLimitDegrees=1}
 {bone1Name=dyn_legArmor_l_01, bone2Name=base_legArmor_l_01, swing1LimitDegrees=25, swing2LimitDegrees=40, twistLimitDegrees=10}
 {
  bone1Name=base_legArmor_l_01
  bone2Name=spine_01
  swing1LimitDegrees=45
  swing1Motion=Free
  swing2LimitDegrees=45
  swing2Motion=Free
  twistLimitDegrees=45
  twistMotion=Free
 }
 {
  bone1Name=base_legArmor_r_01
  bone2Name=spine_01
  swing1LimitDegrees=45
  swing1Motion=Free
  swing2LimitDegrees=45
  swing2Motion=Free
  twistLimitDegrees=45
  twistMotion=Free
 }
 {
  bone1Name=spine_01
  bone2Name=pelvis
  swing1LimitDegrees=45
  swing1Motion=Free
  swing2LimitDegrees=45
  swing2Motion=Free
  twistLimitDegrees=45
  twistMotion=Free
 }
 {
  bone1Name=thigh_l
  bone2Name=pelvis
  swing1LimitDegrees=45
  swing1Motion=Free
  swing2LimitDegrees=45
  swing2Motion=Free
  twistLimitDegrees=45
  twistMotion=Free
 }
 {bone1Name=dyn_legArmor_r_03, bone2Name=dyn_legArmor_r_02, swing1LimitDegrees=5, swing2LimitDegrees=5, twistLimitDegrees=1}
 {bone1Name=thigh_r, bone2Name=pelvis, swing1LimitDegrees=3, swing2LimitDegrees=3, twistLimitDegrees=1}
 {bone1Name=dyn_legArmor_r_02, bone2Name=dyn_legArmor_r_01, swing1LimitDegrees=5, swing2LimitDegrees=10, twistLimitDegrees=1}
 {bone1Name=dyn_legArmor_r_01, bone2Name=base_legArmor_r_01, swing1LimitDegrees=25, swing2LimitDegrees=40, twistLimitDegrees=10}
 {
  bone1Name=thigh_twist_01_l
  bone2Name=thigh_l
  swing1LimitDegrees=45
  swing1Motion=Free
  swing2LimitDegrees=45
  swing2Motion=Free
  twistLimitDegrees=45
  twistMotion=Free
 }
 {
  bone1Name=thigh_twist_01_r
  bone2Name=thigh_r
  swing1LimitDegrees=45
  swing1Motion=Free
  swing2LimitDegrees=45
  swing2Motion=Free
  twistLimitDegrees=45
  twistMotion=Free
 }
]
