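# Post-process style AnimBP for the VelvetDeskFiber character body, correcting and simulating armor, shoulderpads and bomb strap.
# Parent-mesh pose runs through Control Rig, constraints and pre-sim bone offsets, RigidBody chains, strap AnimDynamics, then action and post-sim fixes via cached poses.
# RigidBody chains switch on emote; the non-emote chain only gains weight during tactical sprint.
# EventGraph derives gravity override, targeting and movement states, many driving blended additive bone fixes.
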AnimBlueprint /BRCosmetics/Characters/Player/Male/Medium/Bodies/M_MED_VelvetDeskFiber/Meshes/M_MED_VelvetDeskFiber_AnimBP
target_skeleton=&/M_MED_VelvetDeskFiber_Skeleton
graphs [
 AnimGraph {nodes [
   n1 = Get isStandingStill() -> (isStandingStill:bool)
   "Transform (Modify) Bone - Bone: base_legArmor_l_01"!(ComponentPose, bAlphaBoolEnabled:bool=@n1, Translation="0,-0.5,1", Rotation="-5,5,5", Scale="1,1,1") -> (Pose) {
    alphaInputType=Bool
    rotationMode=BMM_Additive
    rotationSpace=BCS_BoneSpace
    translationMode=BMM_Additive
    translationSpace=BCS_BoneSpace
    alphaBoolBlend {blendInTime=0.1, blendOption=Cubic, blendOutTime=0.2}
    boneToModify {boneName=base_legArmor_l_01}
    rotation=<rot -5 5 5>
    translation {y=-0.5, z=1}
   }
   n2 = Get isCrouchingStill() -> (isCrouchingStill:bool)
   "Transform (Modify) Bone - Bone: dyn_legArmor_l_01"!(ComponentPose, bAlphaBoolEnabled:bool=@n2, Translation="0,-1,1", Rotation="0,0,0", Scale="1,1,1") -> (Pose) {
    alphaInputType=Bool
    rotationMode=BMM_Additive
    rotationSpace=BCS_BoneSpace
    translationMode=BMM_Additive
    translationSpace=BCS_BoneSpace
    alphaBoolBlend {blendInTime=0.5, blendOption=Cubic, blendOutTime=0.5}
    boneToModify {boneName=dyn_legArmor_l_01}
    translation {y=-1, z=1}
   }
   n3 = Get bIsCrouchSprinting()
   n4 = "Transform (Modify) Bone - Bone: base_legArmor_bendF_l_01"!(ComponentPose, bAlphaBoolEnabled:bool=@n3, Translation="0,0,3", Rotation="0,0,0", Scale="1,1,1") -> (Pose) {
    alphaInputType=Bool
    rotationMode=BMM_Additive
    rotationSpace=BCS_BoneSpace
    translationMode=BMM_Additive
    translationSpace=BCS_BoneSpace
    alphaBoolBlend {blendInTime=0.5, blendOption=Cubic, blendOutTime=0.5}
    boneToModify {boneName=base_legArmor_bendF_l_01}
    translation {z=3}
   }
   n5 = "Transform (Modify) Bone - Bone: dyn_legArmor_l_02"!(ComponentPose=@n4, bAlphaBoolEnabled:bool=@n3, Translation="0,0,0", Rotation="-10,0,0", Scale="1,1,1") -> (Pose) {
    alphaInputType=Bool
    rotationMode=BMM_Additive
    rotationSpace=BCS_BoneSpace
    translationMode=BMM_Additive
    translationSpace=BCS_BoneSpace
    alphaBoolBlend {blendInTime=0.5, blendOption=Cubic, blendOutTime=0.5}
    boneToModify {boneName=dyn_legArmor_l_02}
    rotation {pitch=-10}
   }
   "Transform (Modify) Bone - Bone: dyn_legArmor_l_03"!(ComponentPose=@n5, bAlphaBoolEnabled:bool=@n3, Translation="0,0,0", Rotation="-5,0,0", Scale="1,1,1") -> (Pose) {
    alphaInputType=Bool
    rotationMode=BMM_Additive
    rotationSpace=BCS_BoneSpace
    translationMode=BMM_Additive
    translationSpace=BCS_BoneSpace
    alphaBoolBlend {blendInTime=0.5, blendOption=Cubic, blendOutTime=0.5}
    boneToModify {boneName=dyn_legArmor_l_03}
    rotation {pitch=-5}
   }
   Output Pose(Result=Use cached pose 'postSimFixes'()) {layerGroup=None}
   n6 = Control Rig!(
    Source=Copy Pose From Mesh() {bUseAttachedParent=true}
    Alpha
    pub isTargeting float:double=Get isTargeting float() -> (isTargeting float:double)
    pub isUsingHangGlider:bool=Get isUsingHangGliding() -> (isUsingHangGliding:bool)
    pub isPlayingEmote:bool=Get bIsPlayingEmote()
    pub isRBANEnabled:bool=Get bRigidBodyEnabled()
    IsFEIdle:bool=Get isInIdlePoseFrontEnd() -> (isInIdlePoseFrontEnd:bool)
   ) -> (Pose) {
    alphaBoolBlend {blendOption=Linear}
    controlRigAssetReference {blueprintRigClass=&/M_MED_VelvetDeskFiber_CtrlRig.M_MED_VelvetDeskFiber_CtrlRig_C}
   }
   Save cached pose 'controlRigCP'(Pose=@n6) {cacheName=controlRigCP}
   n7 = Anim Dynamics - Bone: dyn_bombStrap(ComponentPose=Local To Component(LocalPose=Use cached pose 'rigidBodiesCP'()), Alpha=1, ExternalForce="0,0,0") {
    angularDampingOverride=1
    angularSpringConstant=20000
    bAngularSpring=true
    bOverrideAngularDamping=true
    bUseGravityOverride=true
    gravityScale=0.35
    simulationSpace=RootRelative
    alphaBoolBlend {blendOption=Linear}
    boundBone {boneName=dyn_bombStrap}
    componentLinearAccScale=<1 1 1>
    componentLinearVelScale=<1 1 1>
    gravityOverride {z=-980}
    physicsBodyDefinitions [{
      boundBone {boneName=dyn_bombStrap}
      boxExtents=<20 30 30>
      constraintSetup {angularLimitsMax=<5 0 5>, angularLimitsMin=<-5 -50 -5>, angularTarget=<1 0 0>}
      localJointOffset=<4 0 0>
     }]
   }
   Save cached pose 'animDynCP'(Pose=Component To Local(ComponentPose=@n7)) {cacheName=animDynCP}
   n8 = Get emoteAlpha() -> (emoteAlpha:double)
   n9 = Get GravityOverride()
   n10 = RigidBody!(ComponentPose=Local To Component(LocalPose=Use cached pose 'presimFixesCP'()), Alpha=@n8, OverrideWorldGravity:Vector=@n9, ExternalForce) -> (Pose) {
    bClampLinearTranslationLimitToRefPose=true
    bOverrideWorldGravity=true
    overridePhysicsAsset=&/M_MED_VelvetDeskFiber_Physics_backFront
    simulationSpace=BaseBoneSpace
    alphaBoolBlend {blendOption=Linear}
    baseBoneRef {boneName=pelvis}
    componentLinearAccScale=<0.4 0.4 0.4>
    componentLinearVelScale=<0.4 0.4 0.4>
    simSpaceSettings {worldAlpha=0.2}
   }
   n11 = RigidBody(ComponentPose=@n10, Alpha=@n8, OverrideWorldGravity=@n9) {
    bClampLinearTranslationLimitToRefPose=true
    bOverrideWorldGravity=true
    overridePhysicsAsset=&/M_MED_VelvetDeskFiber_Physics_sidesEmote
    simulationSpace=BaseBoneSpace
    alphaBoolBlend {blendOption=Linear}
    baseBoneRef {boneName=pelvis}
    componentLinearAccScale=<0.4 0.4 0.4>
    componentLinearVelScale=<0.4 0.4 0.4>
    simSpaceSettings {worldAlpha=0.2}
   }
   n12 = RigidBody(ComponentPose=Local To Component(LocalPose=Use cached pose 'presimFixesCP'()), OverrideWorldGravity=@n9) {
    bClampLinearTranslationLimitToRefPose=true
    bOverrideWorldGravity=true
    overridePhysicsAsset=&/M_MED_VelvetDeskFiber_Physics_backFrontEmote
    simulationSpace=BaseBoneSpace
    alphaBoolBlend {blendOption=Linear}
    baseBoneRef {boneName=pelvis}
    componentLinearAccScale=<0.4 0.4 0.4>
    componentLinearVelScale=<0.4 0.4 0.4>
    simSpaceSettings {worldAlpha=0.2}
   }
   n13 = RigidBody(ComponentPose=@n12, OverrideWorldGravity=@n9) {
    bClampLinearTranslationLimitToRefPose=true
    bOverrideWorldGravity=true
    overridePhysicsAsset=&M_MED_VelvetDeskFiber_Physics_sidesEmote
    simulationSpace=BaseBoneSpace
    alphaBoolBlend {blendOption=Linear}
    baseBoneRef {boneName=pelvis}
    componentAppliedLinearAccClamp=<5000 5000 5000>
    componentLinearAccScale=<0.3 0.3 0.3>
    componentLinearVelScale=<0.3 0.3 0.3>
    simSpaceSettings {maxAngularAcceleration=500, maxAngularVelocity=3, maxLinearAcceleration=400, maxLinearVelocity=250, velocityScaleZ=0.12, worldAlpha=1}
   }
   Save cached pose 'rigidBodiesCP'(Pose=Blend Poses by bool(BlendPose_0=Component To Local(ComponentPose=@n11), BlendPose_1=Component To Local(ComponentPose=@n13), bActiveValue=Get bIsPlayingEmote()) {blendType=Linear}) {cacheName=rigidBodiesCP}
   n14 = "Transform (Modify) Bone - Bone: base_backArmor_01"!(
    ComponentPose=Local To Component(LocalPose=Use cached pose 'animDynCP'())
    bAlphaBoolEnabled:bool=Get bIsSkydiveFloating()
    Translation="0,0,1"
    Rotation="10,0,0"
   ) -> (Pose) {
    alphaInputType=Bool
    rotationMode=BMM_Additive
    rotationSpace=BCS_BoneSpace
    translationMode=BMM_Additive
    translationSpace=BCS_BoneSpace
    alphaBoolBlend {blendInTime=1, blendOption=Cubic, blendOutTime=1}
    boneToModify {boneName=base_backArmor_01}
    rotation {pitch=10}
    translation {z=1}
   }
   n15 = "Transform (Modify) Bone - Bone: attach_cape"!(ComponentPose=@n14, bAlphaBoolEnabled:bool=Get bIsSkydiveDiveMode(), Translation="0,0,0", Rotation="0,0,20") -> (Pose) {
    alphaInputType=Bool
    rotationMode=BMM_Additive
    rotationSpace=BCS_BoneSpace
    translationMode=BMM_Additive
    translationSpace=BCS_BoneSpace
    alphaBoolBlend {blendInTime=2, blendOption=Sinusoidal, blendOutTime=2}
    boneToModify {boneName=attach_cape}
    rotation {roll=20}
   }
   Save cached pose 'actionFixesCP'(Pose=Component To Local(ComponentPose=@n15)) {cacheName=actionFixesCP}
   n16 = "Constraint - dyn_shoulderpad_l (PPPPPPPP)"!(
    ComponentPose=Local To Component(LocalPose=Use cached pose 'controlRigCP'())
    Alpha=1
    ConstraintWeights_0=0.3368
    ConstraintWeights_1:float=0.2727
    ConstraintWeights_2:float=0.1973
    ConstraintWeights_3:float=0.0684
    ConstraintWeights_4:float=0.0547
    ConstraintWeights_5:float=0.0405
    ConstraintWeights_6:float=0.0253
    ConstraintWeights_7:float=0.0043
   ) -> (Pose) {
    alphaBoolBlend {blendOption=Linear}
    boneToModify {boneName=dyn_shoulderpad_l}
    constraintSetup [
     {targetBone {boneName=deform_upperarm_up_l}}
     {targetBone {boneName=upperarm_twist_01_l}}
     {targetBone {boneName=clavicle_l}}
     {targetBone {boneName=upperarm_twist_02_l}}
     {targetBone {boneName=deform_upperarm_fr_l}}
     {targetBone {boneName=spine_05}}
     {targetBone {boneName=spine_03}}
     {targetBone {boneName=deform_upperarm_bk_l}}
    ]
    constraintWeights [1,1,1,1,1,1,1,1]
   }
   n17 = "Constraint - dyn_shoulderpad_r (PPPPPPPP)"!(
    ComponentPose=@n16
    Alpha=1
    ConstraintWeights_0=0.3368
    ConstraintWeights_1:float=0.2727
    ConstraintWeights_2:float=0.1973
    ConstraintWeights_3:float=0.0684
    ConstraintWeights_4:float=0.0547
    ConstraintWeights_5:float=0.0405
    ConstraintWeights_6:float=0.0253
    ConstraintWeights_7:float=0.0043
   ) -> (Pose) {
    alphaBoolBlend {blendOption=Linear}
    boneToModify {boneName=dyn_shoulderpad_r}
    constraintSetup [
     {targetBone {boneName=deform_upperarm_up_r}}
     {targetBone {boneName=upperarm_twist_01_r}}
     {targetBone {boneName=clavicle_r}}
     {targetBone {boneName=upperarm_twist_02_r}}
     {targetBone {boneName=deform_upperarm_fr_r}}
     {targetBone {boneName=spine_05}}
     {targetBone {boneName=spine_03}}
     {targetBone {boneName=deform_upperarm_bk_r}}
    ]
    constraintWeights [1,1,1,1,1,1,1,1]
   }
   n18 = "Constraint - base_backArmor_01 (P)"(ComponentPose=@n17, Alpha=1, ConstraintWeights_0=1) {
    alphaBoolBlend {blendOption=Linear}
    boneToModify {boneName=base_backArmor_01}
    constraintSetup [{targetBone {boneName=spine_02}}]
    constraintWeights [1]
   }
   n19 = "Constraint - base_frontArmor_01 (P)"(ComponentPose=@n18, Alpha=0.5, ConstraintWeights_0=1) {
    alphaBoolBlend {blendOption=Linear}
    boneToModify {boneName=base_frontArmor_01}
    constraintSetup [{targetBone {boneName=spine_02}}]
    constraintWeights [1]
   }
   n20 = "Constraint - dyn_bombStrap (P)"(ComponentPose=@n19, Alpha=1, ConstraintWeights_0=1) {
    alphaBoolBlend {blendOption=Linear}
    boneToModify {boneName=dyn_bombStrap}
    constraintSetup [{targetBone {boneName=dyn_legarmor_top_r}}]
    constraintWeights [1]
   }
   n21 = "Constraint - dyn_legarmor_top_l (P)"(ComponentPose=@n20, Alpha=Get CrouchingWithRigidBodyEnabled() -> (CrouchingWithRigidBodyEnabled:float), ConstraintWeights_0=1) {
    bAlphaBoolEnabled=false
    alphaBoolBlend {blendOption=Linear}
    boneToModify {boneName=dyn_legarmor_top_l}
    constraintSetup [{targetBone {boneName=spine_02}}]
    constraintWeights [1]
   }
   n22 = "Transform (Modify) Bone - Bone: base_frontArmor_01"!(ComponentPose=@n21, bAlphaBoolEnabled:bool=Get bIsDBNO(), Translation="0,0,0", Rotation="10,0,0") -> (Pose) {
    alphaInputType=Bool
    rotationMode=BMM_Additive
    rotationSpace=BCS_BoneSpace
    translationMode=BMM_Additive
    translationSpace=BCS_BoneSpace
    alphaBoolBlend {blendOption=Linear}
    boneToModify {boneName=base_frontArmor_01}
    rotation {pitch=10}
   }
   n23 = "Transform (Modify) Bone - Bone: base_legArmor_r_01"!(ComponentPose=@n22, bAlphaBoolEnabled:bool=True, Translation="0,0,0", Rotation="15,0,-10", Scale="1,1,1") -> (Pose) {
    alphaInputType=Bool
    rotationMode=BMM_Additive
    rotationSpace=BCS_BoneSpace
    translationMode=BMM_Additive
    translationSpace=BCS_BoneSpace
    alphaBoolBlend {blendInTime=0.1, blendOption=Cubic, blendOutTime=0.1}
    boneToModify {boneName=base_legArmor_r_01}
    rotation {pitch=15, roll=-10}
   }
   n24 = "Transform (Modify) Bone - Bone: base_legArmor_l_01"(ComponentPose=@n23, bAlphaBoolEnabled=True, Translation="0,0,0", Rotation="15,0,10", Scale="1,1,1") {
    alphaInputType=Bool
    rotationMode=BMM_Additive
    rotationSpace=BCS_BoneSpace
    translationMode=BMM_Additive
    translationSpace=BCS_BoneSpace
    alphaBoolBlend {blendInTime=0.2, blendOption=Cubic, blendOutTime=0.2}
    boneToModify {boneName=base_legArmor_l_01}
    rotation {pitch=15, roll=10}
   }
   n25 = "Transform (Modify) Bone - Bone: base_frontArmor_01"!(ComponentPose=@n24, bAlphaBoolEnabled=True, Translation="0,0,0", Rotation="-20,0,0", Scale:Vector="1,1,1") -> (Pose) {
    alphaInputType=Bool
    rotationMode=BMM_Additive
    rotationSpace=BCS_BoneSpace
    translationMode=BMM_Additive
    translationSpace=BCS_BoneSpace
    alphaBoolBlend {blendInTime=0.2, blendOption=Cubic, blendOutTime=0.2}
    boneToModify {boneName=base_frontArmor_01}
    rotation {pitch=-20}
   }
   n26 = "Transform (Modify) Bone - Bone: base_backArmor_01"!(ComponentPose=@n25, bAlphaBoolEnabled=True, Translation="0,0,0", Rotation="30,0,0", Scale:Vector="1,1,1") -> (Pose) {
    alphaInputType=Bool
    rotationMode=BMM_Additive
    rotationSpace=BCS_BoneSpace
    translationMode=BMM_Additive
    translationSpace=BCS_BoneSpace
    alphaBoolBlend {blendInTime=0.2, blendOption=Cubic, blendOutTime=0.2}
    boneToModify {boneName=base_backArmor_01}
    rotation {pitch=30}
   }
   n27 = "Transform (Modify) Bone - Bone: base_legArmor_r_01"(ComponentPose=@n26, bAlphaBoolEnabled=True, Translation="0,0,0", Rotation="60,0,0", Scale="1,1,1") {
    alphaInputType=Bool
    rotationMode=BMM_Additive
    rotationSpace=BCS_BoneSpace
    translationMode=BMM_Additive
    translationSpace=BCS_BoneSpace
    alphaBoolBlend {blendInTime=0.2, blendOption=Cubic, blendOutTime=0.2}
    boneToModify {boneName=base_legArmor_r_01}
    rotation {pitch=60}
   }
   n28 = "Transform (Modify) Bone - Bone: base_frontArmor_01"(ComponentPose=@n27, bAlphaBoolEnabled=True, Translation="0,0,2", Rotation="0,0,0", Scale="1,1,1") {
    alphaInputType=Bool
    rotationMode=BMM_Additive
    rotationSpace=BCS_BoneSpace
    translationMode=BMM_Additive
    translationSpace=BCS_BoneSpace
    alphaBoolBlend {blendInTime=0.2, blendOption=Cubic, blendOutTime=0.2}
    boneToModify {boneName=base_frontArmor_01}
    translation {z=2}
   }
   n29 = "Transform (Modify) Bone - Bone: base_frontArmor_01"(ComponentPose=@n28, bAlphaBoolEnabled=True, Translation="0,0,2", Rotation="5,0,0", Scale="1,1,1") {
    alphaInputType=Bool
    rotationMode=BMM_Additive
    rotationSpace=BCS_BoneSpace
    translationMode=BMM_Additive
    translationSpace=BCS_BoneSpace
    alphaBoolBlend {blendInTime=0.2, blendOption=Cubic, blendOutTime=0.2}
    boneToModify {boneName=base_frontArmor_01}
    rotation {pitch=5}
    translation {z=2}
   }
   n30 = "Transform (Modify) Bone - Bone: base_frontArmor_01"(ComponentPose=@n29, bAlphaBoolEnabled=True, Translation="0,0,0", Rotation="5,0,0", Scale="1,1,1") {
    alphaInputType=Bool
    rotationMode=BMM_Additive
    rotationSpace=BCS_BoneSpace
    translationMode=BMM_Additive
    translationSpace=BCS_BoneSpace
    alphaBoolBlend {blendInTime=0.2, blendOption=Cubic, blendOutTime=0.2}
    boneToModify {boneName=base_frontArmor_01}
    rotation {pitch=5}
   }
   n31 = Get isStandingStill()
   n32 = "Transform (Modify) Bone - Bone: dyn_legarmor_top_r"!(ComponentPose=@n30, bAlphaBoolEnabled:bool=@n31, Translation="0,0,0", Rotation="10,10,0", Scale="1,1,1") -> (Pose) {
    alphaInputType=Bool
    rotationMode=BMM_Additive
    rotationSpace=BCS_BoneSpace
    translationMode=BMM_Additive
    translationSpace=BCS_BoneSpace
    alphaBoolBlend {blendInTime=0.1, blendOption=Cubic, blendOutTime=0.2}
    boneToModify {boneName=dyn_legarmor_top_r}
    rotation {pitch=10, yaw=10}
   }
   n33 = "Transform (Modify) Bone - Bone: dyn_legarmor_top_l"!(ComponentPose=@n32, bAlphaBoolEnabled:bool=@n31, Translation="0,0,0", Rotation="10,0,0", Scale="1,1,1") -> (Pose) {
    alphaInputType=Bool
    rotationMode=BMM_Additive
    rotationSpace=BCS_BoneSpace
    translationMode=BMM_Additive
    translationSpace=BCS_BoneSpace
    alphaBoolBlend {blendInTime=0.1, blendOption=Cubic, blendOutTime=0.2}
    boneToModify {boneName=dyn_legarmor_top_l}
    rotation {pitch=10}
   }
   n34 = "Transform (Modify) Bone - Bone: base_frontArmor_01"(ComponentPose=@n33, bAlphaBoolEnabled=@n31, Translation="0,0,0", Rotation="0,0,0", Scale="1,1,1") {
    alphaInputType=Bool
    rotationMode=BMM_Additive
    rotationSpace=BCS_BoneSpace
    translationMode=BMM_Additive
    translationSpace=BCS_BoneSpace
    alphaBoolBlend {blendInTime=0.1, blendOption=Cubic, blendOutTime=0.2}
    boneToModify {boneName=base_frontArmor_01}
   }
   n35 = "Transform (Modify) Bone - Bone: base_frontArmor_01"(ComponentPose=@n34, bAlphaBoolEnabled=Get bIsOnGrindRail(), Translation="0,5,0", Rotation="0,0,0", Scale="1,1,1") {
    alphaInputType=Bool
    rotationMode=BMM_Additive
    rotationSpace=BCS_BoneSpace
    translationMode=BMM_Additive
    translationSpace=BCS_BoneSpace
    alphaBoolBlend {blendInTime=0.2, blendOption=Cubic, blendOutTime=0.2}
    boneToModify {boneName=base_frontArmor_01}
    translation {y=5}
   }
   n36 = "Transform (Modify) Bone - Bone: dyn_legArmor_bendF_l_01"!(ComponentPose=@n35, bAlphaBoolEnabled:bool=Get onZipline() -> (onZipline:bool), Translation="-1,-1,5", Rotation="0,0,0", Scale="1,1,1") -> (Pose) {
    alphaInputType=Bool
    rotationMode=BMM_Additive
    rotationSpace=BCS_BoneSpace
    translationMode=BMM_Additive
    translationSpace=BCS_BoneSpace
    alphaBoolBlend {blendInTime=0.2, blendOption=Cubic, blendOutTime=0.2}
    boneToModify {boneName=dyn_legArmor_bendF_l_01}
    translation=<-1 -1 5>
   }
   Save cached pose 'presimFixesCP'(Pose=Component To Local(ComponentPose=@n36)) {cacheName=presimFixesCP}
   n37 = Get bIsTacticalSprinting()
   n38 = "Transform (Modify) Bone - Bone: dyn_legArmor_l_01"!(
    ComponentPose=Local To Component(LocalPose=Use cached pose 'actionFixesCP'())
    bAlphaBoolEnabled=@n37
    Translation="0,0,0"
    Rotation="-10,0,0"
   ) -> (Pose) {
    alphaInputType=Bool
    rotationMode=BMM_Additive
    rotationSpace=BCS_BoneSpace
    translationMode=BMM_Additive
    translationSpace=BCS_BoneSpace
    alphaBoolBlend {blendInTime=2, blendOption=Linear, blendOutTime=2}
    boneToModify {boneName=dyn_legArmor_l_01}
    rotation {pitch=-10}
   }
   n39 = "Transform (Modify) Bone - Bone: dyn_legArmor_r_01"!(ComponentPose=@n38, bAlphaBoolEnabled:bool=@n37, Translation="0,0,0", Rotation="-10,0,0") -> (Pose) {
    alphaInputType=Bool
    rotationMode=BMM_Additive
    rotationSpace=BCS_BoneSpace
    translationMode=BMM_Additive
    translationSpace=BCS_BoneSpace
    alphaBoolBlend {blendInTime=2, blendOption=Linear, blendOutTime=2}
    boneToModify {boneName=dyn_legArmor_r_01}
    rotation {pitch=-10}
   }
   n40 = "Transform (Modify) Bone - Bone: dyn_legArmor_r_01"!(ComponentPose=@n39, bAlphaBoolEnabled=True, Translation="0,0,0", Rotation="30,0,-10", Scale:Vector="1,1,1") -> (Pose) {
    alphaInputType=Bool
    rotationMode=BMM_Additive
    rotationSpace=BCS_BoneSpace
    translationMode=BMM_Additive
    translationSpace=BCS_BoneSpace
    alphaBoolBlend {blendInTime=0.2, blendOption=Cubic, blendOutTime=0.2}
    boneToModify {boneName=dyn_legArmor_r_01}
    rotation {pitch=30, roll=-10}
   }
   n41 = "Transform (Modify) Bone - Bone: dyn_backArmor_01"!(ComponentPose=@n40, bAlphaBoolEnabled:bool=True, Translation="0,0,0", Rotation="40,0,30", Scale="1,1,1") -> (Pose) {
    alphaInputType=Bool
    rotationMode=BMM_Additive
    rotationSpace=BCS_BoneSpace
    translationMode=BMM_Additive
    translationSpace=BCS_BoneSpace
    alphaBoolBlend {blendInTime=0.2, blendOption=Cubic, blendOutTime=0.2}
    boneToModify {boneName=dyn_backArmor_01}
    rotation {pitch=40, roll=30}
   }
   n42 = "Transform (Modify) Bone - Bone: dyn_backArmor_02"!(ComponentPose=@n41, bAlphaBoolEnabled:bool=True, Translation="0,0,0", Rotation="40,0,0", Scale="1,1,1") -> (Pose) {
    alphaInputType=Bool
    rotationMode=BMM_Additive
    rotationSpace=BCS_BoneSpace
    translationMode=BMM_Additive
    translationSpace=BCS_BoneSpace
    alphaBoolBlend {blendInTime=0.2, blendOption=Cubic, blendOutTime=0.2}
    boneToModify {boneName=dyn_backArmor_02}
    rotation {pitch=40}
   }
   n43 = "Transform (Modify) Bone - Bone: base_backArmor_01"(ComponentPose=@n42, bAlphaBoolEnabled=@n2, Translation="0,0,1", Rotation="-5,0,0", Scale="1,1,1") {
    alphaInputType=Bool
    rotationMode=BMM_Additive
    rotationSpace=BCS_BoneSpace
    translationMode=BMM_Additive
    translationSpace=BCS_BoneSpace
    alphaBoolBlend {blendInTime=0.5, blendOption=Cubic, blendOutTime=0.5}
    boneToModify {boneName=base_backArmor_01}
    rotation {pitch=-5}
    translation {z=1}
   }
   n44 = "Transform (Modify) Bone - Bone: dyn_legarmor_top_r"(ComponentPose=@n43, bAlphaBoolEnabled=@n2, Translation="-2,-1,1", Rotation="4.996,9.984,-10.011", Scale="1,1,1") {
    alphaInputType=Bool
    rotationMode=BMM_Additive
    rotationSpace=BCS_BoneSpace
    translationMode=BMM_Additive
    translationSpace=BCS_BoneSpace
    alphaBoolBlend {blendInTime=0.5, blendOption=Cubic, blendOutTime=0.5}
    boneToModify {boneName=dyn_legarmor_top_r}
    rotation=<rot 4.996 9.984 -10.011>
    translation=<-2 -1 1>
   }
   n45 = "Transform (Modify) Bone - Bone: dyn_legarmor_top_l"(ComponentPose=@n44, bAlphaBoolEnabled=@n2, Translation="-0.781,0.329,3.847", Rotation="33.904,-11.287,-33.492", Scale="1,1,1") {
    alphaInputType=Bool
    rotationMode=BMM_Additive
    rotationSpace=BCS_BoneSpace
    translationMode=BMM_Additive
    translationSpace=BCS_BoneSpace
    alphaBoolBlend {blendInTime=0.5, blendOption=Cubic, blendOutTime=0.5}
    boneToModify {boneName=dyn_legarmor_top_l}
    rotation=<rot 33.904 -11.287 -33.492>
    translation=<-0.781 0.329 3.847>
   }
   n46 = "Transform (Modify) Bone - Bone: base_legArmor_r_01"(ComponentPose=@n45, bAlphaBoolEnabled=Get bIsCrouching(), Translation="0,0,0", Rotation="10,0,0", Scale="1,1,1") {
    alphaInputType=Bool
    rotationMode=BMM_Additive
    rotationSpace=BCS_BoneSpace
    translationMode=BMM_Additive
    translationSpace=BCS_BoneSpace
    alphaBoolBlend {blendInTime=0.5, blendOption=Cubic, blendOutTime=0.5}
    boneToModify {boneName=base_legArmor_r_01}
    rotation {pitch=10}
   }
   n47 = "Transform (Modify) Bone - Bone: dyn_legarmor_top_l"(ComponentPose=@n46, bAlphaBoolEnabled=Get crouchmove() -> (crouchmove:bool), Translation="0,0,0", Rotation="30,-10,-30", Scale="1,1,1") {
    alphaInputType=Bool
    rotationMode=BMM_Additive
    rotationSpace=BCS_BoneSpace
    translationMode=BMM_Additive
    translationSpace=BCS_BoneSpace
    alphaBoolBlend {blendInTime=0.5, blendOption=Cubic, blendOutTime=0.5}
    boneToModify {boneName=dyn_legarmor_top_l}
    rotation=<rot 30 -10 -30>
   }
   n48 = "Transform (Modify) Bone - Bone: base_legArmor_r_01"(ComponentPose=@n47, bAlphaBoolEnabled=@n1, Translation="0,0.5,1", Rotation="0,0,0", Scale="1,1,1") {
    alphaInputType=Bool
    rotationMode=BMM_Additive
    rotationSpace=BCS_BoneSpace
    translationMode=BMM_Additive
    translationSpace=BCS_BoneSpace
    alphaBoolBlend {blendInTime=0.1, blendOption=Cubic, blendOutTime=0.2}
    boneToModify {boneName=base_legArmor_r_01}
    translation {y=0.5, z=1}
   }
   n49 = "Transform (Modify) Bone - Bone: base_backArmor_01"(ComponentPose=@n48, bAlphaBoolEnabled=@n1, Translation="0,0,1", Rotation="0,0,0", Scale="1,1,1") {
    alphaInputType=Bool
    rotationMode=BMM_Additive
    rotationSpace=BCS_BoneSpace
    translationMode=BMM_Additive
    translationSpace=BCS_BoneSpace
    alphaBoolBlend {blendInTime=0.1, blendOption=Cubic, blendOutTime=0.2}
    boneToModify {boneName=base_backArmor_01}
    translation {z=1}
   }
   n50 = "Transform (Modify) Bone - Bone: dyn_bombStrap"!(ComponentPose=@n49, bAlphaBoolEnabled:bool=Get bIsJumping(), Translation="-3,0,2", Rotation="0,20,0", Scale="1,1,1") -> (Pose) {
    alphaInputType=Bool
    rotationMode=BMM_Additive
    rotationSpace=BCS_BoneSpace
    translationMode=BMM_Additive
    translationSpace=BCS_BoneSpace
    alphaBoolBlend {blendInTime=0.3, blendOption=Cubic, blendOutTime=0.5}
    boneToModify {boneName=dyn_bombStrap}
    rotation {yaw=20}
    translation {x=-3, z=2}
   }
   Save cached pose 'postSimFixes'(Pose=Component To Local(ComponentPose=@n50)) {cacheName=postSimFixes}
  ]}
 EventGraph {nodes [
   n1 = Event BlueprintUpdateAnimation()
   > Set Delta Time X(Delta Time X=@n1.DeltaTimeX)
   > n2 = Sequence!(execute) -> (then_0, then_1, then_2:Exec, then_3:Exec, then_4:Exec, then_5:Exec, then_6:Exec)
   n3 = Get bIsTargeting()
   Set isTargeting float(
    execute=@n2.then_1
    isTargeting float:double=FInterpTo(Current=Get isTargeting float(), Target="To Float (Boolean)"(InBool=@n3), DeltaTime=Get Delta Time X(), InterpSpeed=15)
   ) -> (Output_Get:double)
   n4 = Branch(execute=@n2.then_5, Condition=OR Boolean(A=Get bIsPlayingEmote(), B=Get bIsTacticalSprinting()))
   > Set emoteAlpha(emoteAlpha:double=1) -> (Output_Get:double)
   Set emoteAlpha(execute=@n4.else, emoteAlpha=0)
   n5 = Get bIsCrouching()
   n6 = Get bIsJumping()
   n7 = Get bIsFalling()
   n8 = Get bIsMoving2D()
   Set isStandingStill(
    execute=@n2.then_3
    isStandingStill:bool=AND Boolean!(
     A=Get bIsOnGround() -> (bIsOnGround:bool)
     B=NOT Boolean(A=@n5)
     C:bool=NOT Boolean(A=@n8)
     D:bool=NOR Boolean!(A:bool=@n6, B:bool=@n7) -> (ReturnValue:bool)
    ) -> (ReturnValue)
   ) -> (Output_Get:bool)
   n9 = Get bIsSkyDiving()
   n10 = Get bIsPlayingEmote()
   Set isInIdlePoseFrontEnd(
    execute=@n2.then_4
    isInIdlePoseFrontEnd:bool=AND Boolean!(A=Get bIsInFrontEnd(), B=NOT Boolean(A=@n10), C=NOT Boolean(A=@n9)) -> (ReturnValue)
   ) -> (Output_Get:bool)
   n11 = Get bIsMoving2D()
   Set isCrouchingStill(execute=@n2.then_2, isCrouchingStill:bool=AND Boolean!(A=Get bIsCrouching(), B=NOT Boolean(A=@n11)) -> (ReturnValue)) -> (Output_Get:bool)
   > Set crouchmove(crouchmove:bool=AND Boolean(A=Get bIsCrouching(), B=Get bIsMoving2D())) -> (Output_Get:bool)
   n12 = Gravity Override State Function DEPRECATED(
    execute=@n2.then_0
    CharacterPartAnimInstance=Self-Reference()
    GameTimeInSeconds=GetGameTimeInSeconds()
    GravityOverrideParams=Get Gravity Override Params()
   )
   > Set GravityOverride(GravityOverride=@n12)
   n13 = Event OnCharacterCustomizationLoaded()
   > Set HeadMesh(HeadMesh:Skeletal Mesh Component Budgeted Object Reference=Get Array Item with EFort Custom Part Type(Array=@n13.CustomMeshParts, PartType=Head)) -> (Output_Get:Skeletal Mesh Component Budgeted Object Reference)
   Event OnParachuteClosed()
   > Set isUsingHangGliding(isUsingHangGliding:bool=false) -> (Output_Get:bool)
   Event OnParachuteOpened()
   > Set isUsingHangGliding(isUsingHangGliding=false)
   > Switch on EGliderType(Selection=Get PlayerGliderType() -> (PlayerGliderType=HangGlider))
   > Set isUsingHangGliding(isUsingHangGliding=true)
   n14 = Event BlueprintProcessAnimInputEvent()
   > n15 = Branch(Condition=IsNamed(self=@n14.AnimInputEvent, InEventName=aie.Player.Zipline.Entered))
   > Set onZipline(onZipline:bool=1) -> (Output_Get:bool)
   Branch(execute=@n15.else, Condition=IsNamed(self=@n14.AnimInputEvent, InEventName=aie.Player.Zipline.Exited))
   > Set onZipline(onZipline=false)
   NOT Boolean(A=false)
   OR Boolean(A=Get bIsRelaxedLevel1() -> (bIsRelaxedLevel1:bool), B=Get bIsRelaxedLevel2() -> (bIsRelaxedLevel2:bool))
  ]}
]
variables [
 crouchmove {type=boolean}
 Delta Time X {type=number, default=0.0333}
 emoteAlpha {type=number}
 Gravity Override Params {type=GravityOverrideParamsStruct}
 GravityOverride {type=Vector}
 HeadMesh {type=SkeletalMeshComponentBudgeted}
 isCrouchingStill {type=boolean}
 isInIdlePoseFrontEnd {type=boolean}
 isStandingStill {type=boolean}
 isTargeting float {type=number}
 isUsingHangGliding {type=boolean}
 onZipline {type=boolean}
]
